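# Physics asset for the jacket of the large female CasualCherie player body, set up for simulated jacket flap bones.
# Four flap chains (left and right, A and B) each run from a Master capsule through three segments, hanging off spine_04 with swing limits and zero-degree twist limits.
# Capsules on the pelvis, spine, upper arms and forearms form a simplified body, linked by their own constraints, alongside the flaps.
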PhysicsAsset /Game/Characters/Player/Female/Large/Bodies/F_LRG_CasualCherie/Meshes/F_LRG_CasualCherie_Physics_Jacket
bodies {
 dyn_jacketFlap_l_A_01 [{shapeType=Capsule, length=2, radius=3, shapeName=dyn_jacketFlap_l_A_01_capsule, center=<2.7709 0 0>, rotation=<rot 90 0 0>}]
 dyn_jacketFlap_l_A_02 [{shapeType=Capsule, length=2, radius=4, shapeName=dyn_jacketFlap_l_A_02_capsule, center=<2.7709 0 0>, rotation=<rot 90 0 0>}]
 dyn_jacketFlap_l_A_03 [{shapeType=Capsule, length=2, radius=4, shapeName=dyn_jacketFlap_l_A_03_capsule, center=<2.7709 0 0>, rotation=<rot 90 0 0>}]
 dyn_jacketFlap_l_A_Master [{shapeType=Capsule, length=2, radius=2, shapeName=dyn_jacketFlap_l_A_Master_capsule, rotation=<rot 90 0 0>}]
 dyn_jacketFlap_l_B_01 [{shapeType=Capsule, length=2, radius=3, shapeName=dyn_jacketFlap_l_B_01_capsule, center=<2.7709 0 0>, rotation=<rot 90 0 0>}]
 dyn_jacketFlap_l_B_02 [{shapeType=Capsule, length=2, radius=4, shapeName=dyn_jacketFlap_l_B_02_capsule, center=<2.7709 0 0>, rotation=<rot 90 0 0>}]
 dyn_jacketFlap_l_B_03 [{shapeType=Capsule, length=2, radius=4, shapeName=dyn_jacketFlap_l_B_03_capsule, center=<2.7709 0 0>, rotation=<rot 90 0 0>}]
 dyn_jacketFlap_l_B_Master [{shapeType=Capsule, length=2, radius=2, shapeName=dyn_jacketFlap_l_B_Master_capsule, rotation=<rot 90 0 0>}]
 dyn_jacketFlap_r_A_01 [{shapeType=Capsule, length=2, radius=4, shapeName=dyn_jacketFlap_r_A_01_capsule, center=<2.7709 0 0>, rotation=<rot 90 0 0>}]
 dyn_jacketFlap_r_A_02 [{shapeType=Capsule, length=2, radius=4, shapeName=dyn_jacketFlap_r_A_02_capsule, center=<2.7709 0 0>, rotation=<rot 90 0 0>}]
 dyn_jacketFlap_r_A_03 [{shapeType=Capsule, length=2, radius=4, shapeName=dyn_jacketFlap_r_A_03_capsule, center=<2.7709 0 0>, rotation=<rot 90 0 0>}]
 dyn_jacketFlap_r_A_Master [{shapeType=Capsule, length=2, radius=2, shapeName=dyn_jacketFlap_r_A_Master_capsule, rotation=<rot 90 0 0>}]
 dyn_jacketFlap_r_B_01 [{shapeType=Capsule, length=2, radius=4, shapeName=dyn_jacketFlap_r_B_01_capsule, center=<2.7709 0 0>, rotation=<rot 90 0 0>}]
 dyn_jacketFlap_r_B_02 [{shapeType=Capsule, length=2, radius=4, shapeName=dyn_jacketFlap_r_B_02_capsule, center=<2.7709 0 0>, rotation=<rot 90 0 0>}]
 dyn_jacketFlap_r_B_03 [{shapeType=Capsule, length=2, radius=4, shapeName=dyn_jacketFlap_r_B_03_capsule, center=<2.7709 0 0>, rotation=<rot 90 0 0>}]
 dyn_jacketFlap_r_B_Master [{shapeType=Capsule, length=2, radius=2, shapeName=dyn_jacketFlap_r_B_Master_capsule, rotation=<rot 90 0 0>}]
 lowerarm_l [{shapeType=Capsule, length=14.3083, radius=5.3039, center=<9.5186 -0.1429 -0.0901>, rotation=<rot 90 180 180>}]
 lowerarm_r [{shapeType=Capsule, length=14.3083, radius=5.3039, center=<-9.5186 0.1429 0.0901>, rotation=<rot 90 0 180>}]
 pelvis [{shapeType=Capsule, length=10, radius=16, center=<1.266 -2.7241 0.0624>}]
 spine_02 [{shapeType=Capsule, length=6, radius=13, shapeName=spine_02_capsule}]
 spine_03 [{shapeType=Capsule, length=1.5, radius=10, center=<5.724 1.3176 0.0071>}]
 spine_04 [{shapeType=Capsule, length=3, radius=12, center=<2.0779 3.9965 -0.0218>}]
 upperarm_l [{shapeType=Capsule, length=6, radius=3, center=<13.0707 -2.1953 1.7034>, rotation=<rot 90 0 0>}]
 upperarm_r [{shapeType=Capsule, length=6, radius=3, center=<-13.0707 2.1953 -1.7034>, rotation=<rot 90 0 180>}]
}
constraints [
 {bone1Name=spine_03, bone2Name=spine_02, swing1LimitDegrees=45, swing2LimitDegrees=45, twistLimitDegrees=45}
 {bone1Name=spine_04, bone2Name=spine_03, swing1LimitDegrees=45, swing2LimitDegrees=45, twistLimitDegrees=45}
 {bone1Name=upperarm_l, bone2Name=spine_04, swing1LimitDegrees=45, swing2LimitDegrees=45, twistLimitDegrees=45}
 {bone1Name=lowerarm_l, bone2Name=upperarm_l, swing1LimitDegrees=45, swing2LimitDegrees=45, twistLimitDegrees=45}
 {
  bone1Name=spine_02
  bone2Name=pelvis
  swing1LimitDegrees=45
  swing1Motion=Free
  swing2LimitDegrees=45
  swing2Motion=Free
  twistLimitDegrees=45
  twistMotion=Free
 }
 {bone1Name=dyn_jacketFlap_l_A_Master, bone2Name=spine_04, swing1LimitDegrees=45, swing2LimitDegrees=10, twistLimitDegrees=0}
 {bone1Name=dyn_jacketFlap_l_A_01, bone2Name=dyn_jacketFlap_l_A_Master, swing1LimitDegrees=10, swing2LimitDegrees=15, twistLimitDegrees=0}
 {bone1Name=dyn_jacketFlap_l_A_02, bone2Name=dyn_jacketFlap_l_A_01, swing1LimitDegrees=10, swing2LimitDegrees=15, twistLimitDegrees=0}
 {bone1Name=dyn_jacketFlap_l_A_03, bone2Name=dyn_jacketFlap_l_A_02, swing1LimitDegrees=10, swing2LimitDegrees=15, twistLimitDegrees=0}
 {bone1Name=dyn_jacketFlap_r_A_Master, bone2Name=spine_04, swing1LimitDegrees=45, swing2LimitDegrees=10, twistLimitDegrees=0}
 {bone1Name=dyn_jacketFlap_r_A_01, bone2Name=dyn_jacketFlap_r_A_Master, swing1LimitDegrees=10, swing2LimitDegrees=15, twistLimitDegrees=0}
 {bone1Name=dyn_jacketFlap_r_A_02, bone2Name=dyn_jacketFlap_r_A_01, swing1LimitDegrees=10, swing2LimitDegrees=15, twistLimitDegrees=0}
 {bone1Name=dyn_jacketFlap_r_A_03, bone2Name=dyn_jacketFlap_r_A_02, swing1LimitDegrees=10, swing2LimitDegrees=15, twistLimitDegrees=0}
 {bone1Name=dyn_jacketFlap_l_B_Master, bone2Name=spine_04, swing1LimitDegrees=45, swing2LimitDegrees=10, twistLimitDegrees=0}
 {bone1Name=dyn_jacketFlap_l_B_01, bone2Name=dyn_jacketFlap_l_B_Master, swing1LimitDegrees=10, swing2LimitDegrees=15, twistLimitDegrees=0}
 {bone1Name=dyn_jacketFlap_l_B_02, bone2Name=dyn_jacketFlap_l_B_01, swing1LimitDegrees=10, swing2LimitDegrees=15, twistLimitDegrees=0}
 {bone1Name=dyn_jacketFlap_l_B_03, bone2Name=dyn_jacketFlap_l_B_02, swing1LimitDegrees=10, swing2LimitDegrees=15, twistLimitDegrees=0}
 {bone1Name=dyn_jacketFlap_r_B_Master, bone2Name=spine_04, swing1LimitDegrees=10, swing2LimitDegrees=10, twistLimitDegrees=0}
 {bone1Name=dyn_jacketFlap_r_B_01, bone2Name=dyn_jacketFlap_r_B_Master, swing1LimitDegrees=10, swing2LimitDegrees=15, twistLimitDegrees=0}
 {bone1Name=dyn_jacketFlap_r_B_02, bone2Name=dyn_jacketFlap_r_B_01, swing1LimitDegrees=10, swing2LimitDegrees=15, twistLimitDegrees=0}
 {bone1Name=dyn_jacketFlap_r_B_03, bone2Name=dyn_jacketFlap_r_B_02, swing1LimitDegrees=10, swing2LimitDegrees=15, twistLimitDegrees=0}
 {bone1Name=lowerarm_r, bone2Name=upperarm_r, swing1LimitDegrees=45, swing2LimitDegrees=45, twistLimitDegrees=45}
 {bone1Name=upperarm_r, bone2Name=spine_04, swing1LimitDegrees=45, swing2LimitDegrees=45, twistLimitDegrees=45}
]
