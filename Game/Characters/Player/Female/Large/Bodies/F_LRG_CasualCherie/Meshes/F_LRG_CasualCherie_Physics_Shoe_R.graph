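# Physics asset for the right shoe loop on the large female CasualCherie player body, giving the loop bones physical bodies for simulated movement.
# Four small capsules form a chain from the Master bone through bones 01, 02 and 03, each constrained to the one before it.
# Every joint shares narrow, asymmetric swing limits and a zero twist limit.
# The capsule shape names carry left-side labels, although the bodies sit on right-side bones.
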PhysicsAsset /Game/Characters/Player/Female/Large/Bodies/F_LRG_CasualCherie/Meshes/F_LRG_CasualCherie_Physics_Shoe_R
bodies {
 dyn_shoeloop_r_01 [{shapeType=Capsule, length=0.5, radius=1, shapeName=dyn_shoeloop_l_01_capsule, center=<1.2338 0 0>, rotation=<rot 90 0 0>}]
 dyn_shoeloop_r_02 [{shapeType=Capsule, length=0.5, radius=1, shapeName=dyn_shoeloop_l_02_capsule, center=<1.2338 0 0>, rotation=<rot 90 0 0>}]
 dyn_shoeloop_r_03 [{shapeType=Capsule, length=0.5, radius=1, shapeName=dyn_shoeloop_l_03_capsule, center=<1.2338 0 0>, rotation=<rot 90 0 0>}]
 dyn_shoeloop_r_Master [{shapeType=Capsule, length=0.5, radius=1, shapeName=dyn_shoeloop_l_Master_capsule, center=<1.2412 0 0>, rotation=<rot 90 0 0>}]
}
constraints [
 {bone1Name=dyn_shoeloop_r_01, bone2Name=dyn_shoeloop_r_Master, swing1LimitDegrees=10, swing2LimitDegrees=30, twistLimitDegrees=0}
 {bone1Name=dyn_shoeloop_r_02, bone2Name=dyn_shoeloop_r_01, swing1LimitDegrees=10, swing2LimitDegrees=30, twistLimitDegrees=0}
 {bone1Name=dyn_shoeloop_r_03, bone2Name=dyn_shoeloop_r_02, swing1LimitDegrees=10, swing2LimitDegrees=30, twistLimitDegrees=0}
]
