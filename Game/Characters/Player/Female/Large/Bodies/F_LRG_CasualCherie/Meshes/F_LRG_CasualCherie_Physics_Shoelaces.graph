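# Physics asset for the left-shoe laces on the large female CasualCherie player body, giving two lace strands and a loop their physical motion.
# Each is a chain of a master and three capsule segments, with the master constrained to the left calf twist bone with free rotation.
# Segments allow limited swing and zero twist, tighter on the loop than the laces.
# Larger calf and foot capsules supply the leg's shape, and the foot has no constraint.
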PhysicsAsset /Game/Characters/Player/Female/Large/Bodies/F_LRG_CasualCherie/Meshes/F_LRG_CasualCherie_Physics_Shoelaces
bodies {
 calf_twist_01_l [{shapeType=Capsule, length=9, radius=5, shapeName=calf_twist_01_l_capsule, rotation=<rot 90 0 0>}]
 dyn_shoelace_l_A_01 [{shapeType=Capsule, length=2, radius=1, shapeName=dyn_shoelace_l_A_01_capsule, center=<2.0939 0 0>, rotation=<rot 90 0 0>}]
 dyn_shoelace_l_A_02 [{shapeType=Capsule, length=2, radius=1, shapeName=dyn_shoelace_l_A_02_capsule, center=<2.0939 0 0>, rotation=<rot 90 0 0>}]
 dyn_shoelace_l_A_03 [{shapeType=Capsule, length=2, radius=1, shapeName=dyn_shoelace_l_A_03_capsule, center=<2.0939 0 0>, rotation=<rot 90 0 0>}]
 dyn_shoelace_l_A_Master [{shapeType=Capsule, length=0.505, radius=0.505, shapeName=dyn_shoelace_l_A_Master_capsule, rotation=<rot 0 -90 0>}]
 dyn_shoelace_l_B_01 [{shapeType=Capsule, length=2, radius=1, shapeName=dyn_shoelace_l_B_01_capsule, center=<2.0939 0 0>, rotation=<rot 90 0 0>}]
 dyn_shoelace_l_B_02 [{shapeType=Capsule, length=2, radius=1, shapeName=dyn_shoelace_l_B_02_capsule, center=<2.0939 0 0>, rotation=<rot 90 0 0>}]
 dyn_shoelace_l_B_03 [{shapeType=Capsule, length=2, radius=1, shapeName=dyn_shoelace_l_B_03_capsule, center=<2.0939 0 0>, rotation=<rot 90 0 0>}]
 dyn_shoelace_l_B_Master [{shapeType=Capsule, length=0.505, radius=0.505, shapeName=dyn_shoelace_l_B_Master_capsule, rotation=<rot 0 -90 0>}]
 dyn_shoeloop_l_01 [{shapeType=Capsule, length=0.5, radius=1, shapeName=dyn_shoeloop_l_01_capsule, center=<1.3261 0 0>, rotation=<rot 90 0 0>}]
 dyn_shoeloop_l_02 [{shapeType=Capsule, length=0.5, radius=1, shapeName=dyn_shoeloop_l_02_capsule, center=<1.3261 0 0>, rotation=<rot 90 0 0>}]
 dyn_shoeloop_l_03 [{shapeType=Capsule, length=0.5, radius=1, shapeName=dyn_shoeloop_l_03_capsule, center=<1.3261 0 0>, rotation=<rot 90 0 0>}]
 dyn_shoeloop_l_Master [{shapeType=Capsule, length=0.5, radius=1, shapeName=dyn_shoeloop_l_Master_capsule, center=<1.3261 0 0>, rotation=<rot 90 0 0>}]
 foot_l [{shapeType=Capsule, length=9, radius=6.5, shapeName=foot_l_capsule, rotation=<rot 90 -3.6023 42.291>}]
}
constraints [
 {bone1Name=dyn_shoelace_l_A_01, bone2Name=dyn_shoelace_l_A_Master, swing1LimitDegrees=45, swing2LimitDegrees=45, twistLimitDegrees=0}
 {bone1Name=dyn_shoelace_l_A_02, bone2Name=dyn_shoelace_l_A_01, swing1LimitDegrees=45, swing2LimitDegrees=45, twistLimitDegrees=0}
 {bone1Name=dyn_shoelace_l_A_03, bone2Name=dyn_shoelace_l_A_02, swing1LimitDegrees=45, swing2LimitDegrees=45, twistLimitDegrees=0}
 {bone1Name=dyn_shoelace_l_B_01, bone2Name=dyn_shoelace_l_B_Master, swing1LimitDegrees=45, swing2LimitDegrees=45, twistLimitDegrees=0}
 {bone1Name=dyn_shoelace_l_B_02, bone2Name=dyn_shoelace_l_B_01, swing1LimitDegrees=45, swing2LimitDegrees=45, twistLimitDegrees=0}
 {bone1Name=dyn_shoelace_l_B_03, bone2Name=dyn_shoelace_l_B_02, swing1LimitDegrees=45, swing2LimitDegrees=45, twistLimitDegrees=0}
 {
  bone1Name=dyn_shoelace_l_A_Master
  bone2Name=calf_twist_01_l
  swing1LimitDegrees=45
  swing1Motion=Free
  swing2LimitDegrees=45
  swing2Motion=Free
  twistLimitDegrees=45
  twistMotion=Free
 }
 {
  bone1Name=dyn_shoelace_l_B_Master
  bone2Name=calf_twist_01_l
  swing1LimitDegrees=45
  swing1Motion=Free
  swing2LimitDegrees=45
  swing2Motion=Free
  twistLimitDegrees=45
  twistMotion=Free
 }
 {
  bone1Name=dyn_shoeloop_l_Master
  bone2Name=calf_twist_01_l
  swing1LimitDegrees=45
  swing1Motion=Free
  swing2LimitDegrees=45
  swing2Motion=Free
  twistLimitDegrees=45
  twistMotion=Free
 }
 {bone1Name=dyn_shoeloop_l_01, bone2Name=dyn_shoeloop_l_Master, swing1LimitDegrees=10, swing2LimitDegrees=30, twistLimitDegrees=0}
 {bone1Name=dyn_shoeloop_l_02, bone2Name=dyn_shoeloop_l_01, swing1LimitDegrees=10, swing2LimitDegrees=30, twistLimitDegrees=0}
 {bone1Name=dyn_shoeloop_l_03, bone2Name=dyn_shoeloop_l_02, swing1LimitDegrees=10, swing2LimitDegrees=30, twistLimitDegrees=0}
]
