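# Corrective Control Rig for the F_LRG_CasualCherie face accessory mesh, posing the dyn_hair master bones in response to head and clavicle motion.
# Three spherical pose readers measure head direction against spine_05, scaling rotations and X-axis shrink on the center, left and right hair masters.
# Two clavicle readers add side-specific hair rotation, multiplied by the matching head reader so it applies only when both poses are active.
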ControlRigBlueprint /Game/Characters/Player/Female/Large/Bodies/F_LRG_CasualCherie/Meshes/Parts/F_LRG_CasualCherie_FaceAcc_CtrlRig
graphs [
 Multiply_5_1_ContainedGraph {nodes [
   n1 = RigVMFunctionEntryNode()
   RigVMFunctionReturnNode(Result=&RigVMFunction_MathFloatMul(B=@n1.C, A=&RigVMFunction_MathFloatMul(A=@n1.A, B=@n1.B) {template="Multiply::Execute(in A,in B,out Result)"}.Result) {template="Multiply::Execute(in A,in B,out Result)"}.Result)
  ]}
 Multiply_5_1_ContainedGraph {nodes [
   n1 = RigVMFunctionEntryNode()
   RigVMFunctionReturnNode(Result=&RigVMFunction_MathFloatMul(B=@n1.C, A=&RigVMFunction_MathFloatMul(A=@n1.A, B=@n1.B) {template="Multiply::Execute(in A,in B,out Result)"}.Result) {template="Multiply::Execute(in A,in B,out Result)"}.Result)
  ]}
 RigVMFunctionLibrary {links [], nodes []}
 RigVMModel {nodes [
   RigUnit_BeginExecution = &RigUnit_BeginExecution()
   SequenceAggregate = RigVMAggregateNode(ExecuteContext=@RigUnit_BeginExecution.ExecutePin)
   n1 = &RigUnit_SphericalPoseReader(
    ActiveRegionScaleFactors=(PositiveWidth=0.1,NegativeWidth=0.1,PositiveHeight=0.1,NegativeHeight=0.1)
    ActiveRegionSize=0.1
    Debug=(bDrawDebug=true,DebugScale=25,DebugSegments=20,DebugThickness=0.25)
    DriverAxis=(X=0,Y=-1,Z=0)
    DriverItem=(Type=Bone,Name="head")
    FalloffRegionScaleFactors=(PositiveWidth=1,NegativeWidth=1,PositiveHeight=1,NegativeHeight=1)
    FalloffSize=0.5
    FlipWidthScaling=true
    OptionalParentItem=(Type=Bone,Name="spine_05")
    RotationOffset=(X=90,Y=180,Z=90)
    ExecutePin=@SequenceAggregate.A
   )
   Multiply = &RigVMFunction_MathDoubleMul(A=-120, B=@n1.OutputParam)
   n2 = &RigVMFunction_MathDoubleMul(A=-60, B=@n1.OutputParam)
   n3 = &RigVMFunction_MathDoubleMul(A=0, B=@n1.OutputParam)
   n4 = &RigVMFunction_MathDoubleMul(A=-120, B=@n1.OutputParam)
   n5 = &RigVMFunction_MathDoubleMul(A=0, B=@n1.OutputParam)
   n6 = &RigVMFunction_MathDoubleSub(A=1, B=&RigVMFunction_MathFloatMul(A=0, B=@n1.OutputParam) {template="Multiply::Execute(in A,in B,out Result)"}.Result)
   n7 = &RigVMFunction_MathDoubleSub(A=1, B=&RigVMFunction_MathFloatMul(A=0, B=@n1.OutputParam) {template="Multiply::Execute(in A,in B,out Result)"}.Result)
   > &RigUnit_ModifyTransforms(
    ItemToModify="((Item=(Type=Bone,Name=\"dyn_hair_c_Master\"),Transform=(Rotation=(X=0,Y=0,Z=0,W=1),Translation=(X=0,Y=0,Z=0),Scale3D=(X=1,Y=1,Z=1))),(Item=(Type=Bone,Name=\"dyn_hair_l_B_Master\"),Transform=(Rotation=(X=0,Y=0,Z=0,W=1),Translation=(X=0,Y=0,Z=0),Scale3D=(X=1,Y= ...(truncated)"
    Weight=1
    ItemToModify.0.Transform.Rotation=&RigVMFunction_MathQuaternionFromEuler(Euler=(X=0,Y=0,Z=0), Euler.Y=@Multiply.Result).Result
    ItemToModify.3.Transform.Rotation=&RigVMFunction_MathQuaternionFromEuler(Euler=(X=0,Y=0,Z=0), Euler.Y=@n2.Result, Euler.Z=@n3.Result).Result
    ItemToModify.2.Transform.Rotation=&RigVMFunction_MathQuaternionFromEuler(Euler=(X=0,Y=0,Z=0), Euler.Y=@n4.Result, Euler.Z=@n5.Result).Result
    ItemToModify.1.Transform.Rotation=&RigVMFunction_MathQuaternionFromEuler(Euler=(X=0,Y=0,Z=0), Euler.Y=@n4.Result, Euler.Z=&RigVMFunction_MathDoubleMul(B=-1, A=@n5.Result).Result).Result
    ItemToModify.4.Transform.Rotation=&RigVMFunction_MathQuaternionFromEuler(Euler=(X=0,Y=0,Z=0), Euler.Y=@n2.Result, Euler.Z=&RigVMFunction_MathDoubleMul(B=-1, A=@n3.Result).Result).Result
    ItemToModify.3.Transform.Scale3D.X=@n6.Result
    ItemToModify.4.Transform.Scale3D.X=@n6.Result
    ItemToModify.1.Transform.Scale3D.X=@n7.Result
    ItemToModify.2.Transform.Scale3D.X=@n7.Result
   )
   n8 = &RigUnit_SphericalPoseReader(
    ActiveRegionScaleFactors=(PositiveWidth=0.1,NegativeWidth=0.1,PositiveHeight=0.1,NegativeHeight=0.1)
    ActiveRegionSize=0.1
    Debug=(bDrawDebug=true,DebugScale=25,DebugSegments=20,DebugThickness=0.25)
    DriverAxis=(X=0,Y=0,Z=1)
    DriverItem=(Type=Bone,Name="head")
    FalloffRegionScaleFactors=(PositiveWidth=1,NegativeWidth=1,PositiveHeight=1,NegativeHeight=1)
    FalloffSize=0.5
    FlipWidthScaling=true
    OptionalParentItem=(Type=Bone,Name="spine_05")
    RotationOffset=(X=0,Y=90,Z=0)
    ExecutePin=@SequenceAggregate.B
   )
   n9 = &RigVMFunction_MathQuaternionFromEuler(Euler=(X=0,Y=0,Z=0), Euler.Y=&RigVMFunction_MathDoubleMul(A=45, B=@n8.OutputParam).Result)
   Subtract = &RigVMFunction_MathDoubleSub(A=1, B=&RigVMFunction_MathFloatMul(A=0.5, B=@n8.OutputParam) {template="Multiply::Execute(in A,in B,out Result)"}.Result)
   > &RigUnit_ModifyTransforms(
    ItemToModify="((Item=(Type=Bone,Name=\"dyn_hair_r_A_Master\"),Transform=(Rotation=(X=0,Y=0,Z=0,W=1),Translation=(X=0,Y=0,Z=0),Scale3D=(X=1,Y=1,Z=1))),(Item=(Type=Bone,Name=\"dyn_hair_r_B_Master\"),Transform=(Rotation=(X=0,Y=0,Z=0,W=1),Translation=(X=0,Y=0,Z=0),Scale3D=(X=1, ...(truncated)"
    Weight=1
    ItemToModify.0.Transform.Rotation=&RigVMFunction_MathQuaternionFromEuler(Euler=(X=0,Y=0,Z=0), Euler.Y=&RigVMFunction_MathDoubleMul(A=-150, B=@n8.OutputParam).Result).Result
    ItemToModify.2.Transform.Rotation=&RigVMFunction_MathQuaternionFromEuler(
     Euler=(X=0,Y=0,Z=0)
     Euler.Y=&RigVMFunction_MathDoubleMul(A=-120, B=@n8.OutputParam).Result
     Euler.X=&RigVMFunction_MathDoubleMul(A=-90, B=@n8.OutputParam).Result
    ).Result
    ItemToModify.1.Transform.Rotation=&RigVMFunction_MathQuaternionFromEuler(
     Euler=(X=0,Y=0,Z=0)
     Euler.Y=&RigVMFunction_MathDoubleMul(A=-150, B=@n8.OutputParam).Result
     Euler.Z=&RigVMFunction_MathDoubleMul(A=60, B=@n8.OutputParam).Result
    ).Result
    ItemToModify.3.Transform.Rotation=@n9.Result
    ItemToModify.4.Transform.Rotation=@n9.Result
    ItemToModify.0.Transform.Scale3D.X=@Subtract.Result
   )
   n10 = &RigUnit_SphericalPoseReader(
    ActiveRegionScaleFactors=(PositiveWidth=0.1,NegativeWidth=0.1,PositiveHeight=0.1,NegativeHeight=0.1)
    ActiveRegionSize=0.1
    Debug=(bDrawDebug=true,DebugScale=25,DebugSegments=20,DebugThickness=0.25)
    DriverAxis=(X=0,Y=0,Z=-1)
    DriverItem=(Type=Bone,Name="head")
    FalloffRegionScaleFactors=(PositiveWidth=1,NegativeWidth=1,PositiveHeight=1,NegativeHeight=1)
    FalloffSize=0.5
    FlipWidthScaling=true
    OptionalParentItem=(Type=Bone,Name="spine_05")
    RotationOffset=(X=180,Y=-90,Z=0)
    ExecutePin=@SequenceAggregate.C
   )
   > &RigUnit_ModifyTransforms(
    ItemToModify="((Item=(Type=Bone,Name=\"dyn_hair_l_A_Master\"),Transform=(Rotation=(X=0,Y=0,Z=0,W=1),Translation=(X=0,Y=0,Z=0),Scale3D=(X=1,Y=1,Z=1))),(Item=(Type=Bone,Name=\"dyn_hair_l_B_Master\"),Transform=(Rotation=(X=0,Y=0,Z=0,W=1),Translation=(X=0,Y=0,Z=0),Scale3D=(X=1, ...(truncated)"
    Weight=1
    ItemToModify.0.Transform.Rotation=&RigVMFunction_MathQuaternionFromEuler(Euler=(X=0,Y=0,Z=0), Euler.Y=&RigVMFunction_MathDoubleMul(A=-150, B=@n10.OutputParam).Result).Result
    ItemToModify.2.Transform.Rotation=&RigVMFunction_MathQuaternionFromEuler(
     Euler=(X=0,Y=0,Z=0)
     Euler.Y=&RigVMFunction_MathDoubleMul(A=-120, B=@n10.OutputParam).Result
     Euler.X=&RigVMFunction_MathDoubleMul(A=90, B=@n10.OutputParam).Result
    ).Result
    ItemToModify.1.Transform.Rotation=&RigVMFunction_MathQuaternionFromEuler(
     Euler=(X=0,Y=0,Z=0)
     Euler.Y=&RigVMFunction_MathDoubleMul(A=-150, B=@n10.OutputParam).Result
     Euler.Z=&RigVMFunction_MathDoubleMul(A=-60, B=@n10.OutputParam).Result
    ).Result
    ItemToModify.3.Transform.Rotation=&RigVMFunction_MathQuaternionFromEuler(Euler=(X=0,Y=0,Z=0), Euler.Y=&RigVMFunction_MathDoubleMul(A=45, B=@n10.OutputParam).Result).Result
    ItemToModify.0.Transform.Scale3D.X=&RigVMFunction_MathDoubleSub(A=1, B=&RigVMFunction_MathFloatMul(A=0.75, B=@n10.OutputParam) {template="Multiply::Execute(in A,in B,out Result)"}.Result).Result
   )
   n11 = &RigUnit_SphericalPoseReader(
    ActiveRegionScaleFactors=(PositiveWidth=2.5,NegativeWidth=3,PositiveHeight=3,NegativeHeight=3)
    ActiveRegionSize=0.1
    Debug=(bDrawDebug=false,DebugScale=25,DebugSegments=20,DebugThickness=0.25)
    DriverAxis=(X=-1,Y=0,Z=0)
    DriverItem=(Type=Bone,Name="clavicle_r")
    FalloffRegionScaleFactors=(PositiveWidth=1,NegativeWidth=1,PositiveHeight=1,NegativeHeight=1)
    FalloffSize=0.5
    FlipHeightScaling=true
    OptionalParentItem=(Type=Bone)
    RotationOffset=(X=180,Y=0,Z=180)
    ExecutePin=@SequenceAggregate.D
   )
   Multiply_5 = RigVMAggregateNode(A=-60, C=@n8.OutputParam, B=@n11.OutputParam)
   RigVMFunction_MathQuaternionFromEuler = &RigVMFunction_MathQuaternionFromEuler(Euler=(X=0,Y=0,Z=0), Euler.Y=@Multiply_5.Result)
   > &RigUnit_ModifyTransforms(
    ItemToModify="((Item=(Type=Bone,Name=\"dyn_hair_r_A_Master\"),Transform=(Rotation=(X=0,Y=0,Z=0,W=1),Translation=(X=0,Y=0,Z=0),Scale3D=(X=1,Y=1,Z=1))),(Item=(Type=Bone,Name=\"dyn_hair_r_B_Master\"),Transform=(Rotation=(X=0,Y=0,Z=0,W=1),Translation=(X=0,Y=0,Z=0),Scale3D=(X=1, ...(truncated)"
    Weight=1
    ItemToModify.0.Transform.Rotation=@RigVMFunction_MathQuaternionFromEuler.Result
    ItemToModify.1.Transform.Rotation=@RigVMFunction_MathQuaternionFromEuler.Result
   )
   n12 = &RigUnit_SphericalPoseReader(
    ActiveRegionScaleFactors=(PositiveWidth=2.5,NegativeWidth=3,PositiveHeight=3,NegativeHeight=3)
    ActiveRegionSize=0.1
    Debug=(bDrawDebug=false,DebugScale=25,DebugSegments=20,DebugThickness=0.25)
    DriverAxis=(X=1,Y=0,Z=0)
    DriverItem=(Type=Bone,Name="clavicle_l")
    FalloffRegionScaleFactors=(PositiveWidth=1,NegativeWidth=1,PositiveHeight=1,NegativeHeight=1)
    FalloffSize=0.5
    OptionalParentItem=(Type=Bone,Name="None")
    RotationOffset=(X=0,Y=0,Z=0)
    ExecutePin=@SequenceAggregate.E
   )
   Multiply_5_1 = RigVMAggregateNode(A=-60, B=@n12.OutputParam, C=@n10.OutputParam)
   n13 = &RigVMFunction_MathQuaternionFromEuler(Euler=(X=0,Y=0,Z=0), Euler.Y=@Multiply_5_1.Result)
   > &RigUnit_ModifyTransforms(
    ItemToModify="((Item=(Type=Bone,Name=\"dyn_hair_l_A_Master\"),Transform=(Rotation=(X=0,Y=0,Z=0,W=1),Translation=(X=0,Y=0,Z=0),Scale3D=(X=1,Y=1,Z=1))),(Item=(Type=Bone,Name=\"dyn_hair_l_B_Master\"),Transform=(Rotation=(X=0,Y=0,Z=0,W=1),Translation=(X=0,Y=0,Z=0),Scale3D=(X=1, ...(truncated)"
    Weight=1
    ItemToModify.0.Transform.Rotation=@n13.Result
    ItemToModify.1.Transform.Rotation=@n13.Result
   )
  ]}
 SequenceAggregate_1_ContainedGraph {nodes [
   n1 = RigVMFunctionEntryNode()
   SequenceAggregate = &RigVMFunction_Sequence(ExecuteContext=@n1.ExecuteContext)
   SequenceAggregate_1 = &RigVMFunction_Sequence(ExecuteContext=@SequenceAggregate.B)
   SequenceAggregate_2 = &RigVMFunction_Sequence(ExecuteContext=@SequenceAggregate_1.B)
   SequenceAggregate_3 = &RigVMFunction_Sequence(ExecuteContext=@SequenceAggregate_2.B)
   RigVMFunctionReturnNode(A=@SequenceAggregate.A, B=@SequenceAggregate_1.A, C=@SequenceAggregate_2.A, D=@SequenceAggregate_3.A, E=@SequenceAggregate_3.B)
  ]}
]
hierarchy [
 root {loc=<0 0 0>, rot=<0 0 0 1>}
 pelvis {parent=root, loc=<0 -0.648 94.7699>, rot=<0.0127 0.707 -0.0127 -0.707>}
 spine_01 {parent=pelvis, loc=<5.0975 -0.6388 0>, rot=<0 0 0.034 -0.9994>}
 spine_02 {parent=spine_01, loc=<7.2518 -0.035 0>, rot=<0 0 0.0238 -0.9997>}
 spine_03 {parent=spine_02, loc=<6.0104 0 0>, rot=<0 0 0.0149 -0.9999>}
 spine_04 {parent=spine_03, loc=<8.1024 0 0>, rot=<0 0 -0.0132 -0.9999>}
 spine_05 {parent=spine_04, loc=<10.0146 0 0>, rot=<0 0 -0.0015 -1>}
 neck_01 {parent=spine_05, loc=<11.3775 -0.6396 0>, rot=<0 0 -0.2003 -0.9797>}
 neck_02 {parent=neck_01, loc=<4.5627 0 0>, rot=<0 0 0.0504 -0.9987>}
 head {parent=neck_02, loc=<5.3914 0 0>, rot=<0 0 0.115 -0.9934>}
 dyn_hair_l_A_Master {parent=head, loc=<4.9831 -1.4218 -7.8619>, rot=<0 -1 0 0>}
 dyn_hair_l_A_01 {parent=dyn_hair_l_A_Master, loc=<0 0 0>, rot=<0 0 0 1>}
 dyn_hair_l_A_02 {parent=dyn_hair_l_A_01, loc=<5.1518 0 0>, rot=<0 0 0 1>}
 dyn_hair_l_A_03 {parent=dyn_hair_l_A_02, loc=<5.7808 0 0>, rot=<0 0 0 1>}
 dyn_hair_l_B_Master {parent=head, loc=<3.7937 -7.006 -5.2591>, rot=<0 -0.9063 0.4226 0>}
 dyn_hair_l_B_01 {parent=dyn_hair_l_B_Master, loc=<0 0 0>, rot=<0 0 0 1>}
 dyn_hair_l_B_02 {parent=dyn_hair_l_B_01, loc=<4.5618 0 0>, rot=<0 0 0 1>}
 dyn_hair_l_B_03 {parent=dyn_hair_l_B_02, loc=<4.2466 0 0>, rot=<0 0 0 1>}
 dyn_hair_c_Master {parent=head, loc=<3.6124 -9.6163 -0.3076>, rot=<0 -0.7071 0.7071 0>}
 dyn_hair_c_01 {parent=dyn_hair_c_Master, loc=<0 0 0>, rot=<0 0 0 1>}
 dyn_hair_c_02 {parent=dyn_hair_c_01, loc=<4.5618 0 0>, rot=<0 0 0 1>}
 dyn_hair_c_03 {parent=dyn_hair_c_02, loc=<4.2466 0 0>, rot=<0 0 0 1>}
 dyn_hair_r_A_Master {parent=head, loc=<4.9827 -1.4218 7.8619>, rot=<0 0 1 0>}
 dyn_hair_r_A_01 {parent=dyn_hair_r_A_Master, loc=<0 0 0>, rot=<0 0 0 -1>}
 dyn_hair_r_A_02 {parent=dyn_hair_r_A_01, loc=<5.152 0 0>, rot=<0 0 0 -1>}
 dyn_hair_r_A_03 {parent=dyn_hair_r_A_02, loc=<5.781 0 0>, rot=<0 0 0 -1>}
 dyn_hair_r_B_Master {parent=head, loc=<3.7938 -7.006 5.2591>, rot=<0 -0.2588 0.9659 0>}
 dyn_hair_r_B_01 {parent=dyn_hair_r_B_Master, loc=<0 0 0>, rot=<0 0 0 -1>}
 dyn_hair_r_B_02 {parent=dyn_hair_r_B_01, loc=<4.562 0 0>, rot=<0 0 0 -1>}
 dyn_hair_r_B_03 {parent=dyn_hair_r_B_02, loc=<4.247 0 0>, rot=<0 0 0 -1>}
 clavicle_l {parent=spine_05, loc=<4.1651 -0.3023 -5.0942>, rot=<0.0465 -0.7289 -0.009 -0.6829>}
 upperarm_l {parent=clavicle_l, loc=<11.2117 -1.1486 -0.0332>, rot=<0.0179 -0.2982 -0.028 -0.9539>}
 clavicle_r {parent=spine_05, loc=<4.1651 -0.3023 5.0942>, rot=<0.7289 0.0465 -0.6829 0.009>}
 upperarm_r {parent=clavicle_r, loc=<-11.2117 1.1486 0.0332>, rot=<0.0179 -0.2982 -0.028 -0.9539>}
]
