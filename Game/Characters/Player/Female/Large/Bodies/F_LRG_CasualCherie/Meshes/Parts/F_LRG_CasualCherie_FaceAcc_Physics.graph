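# Physics asset for the FaceAcc mesh part of the female large CasualCherie player character, covering five dynamic hair chains.
# Center, left A/B and right A/B chains each link a small master capsule and three segment capsules to the head.
# Segment joints use tight swing limits and zero-degree twist, while master and body joints rotate freely.
# Head, upper spine, clavicle and upper-arm capsules supply body shapes around the hair.
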PhysicsAsset /Game/Characters/Player/Female/Large/Bodies/F_LRG_CasualCherie/Meshes/Parts/F_LRG_CasualCherie_FaceAcc_Physics
bodies {
 clavicle_l [{shapeType=Capsule, length=5, radius=9.5, shapeName=clavicle_l_capsule, center=<2.0978 0 -2.0398>, rotation=<rot 90 0 180>}]
 clavicle_r [{shapeType=Capsule, length=5, radius=9.5, shapeName=clavicle_l_capsule, center=<-2.0978 0 2.0398>, rotation=<rot 90 0 0>}]
 dyn_hair_c_01 [{shapeType=Capsule, length=2, radius=2.5, shapeName=dyn_hair_c_01_capsule, rotation=<rot 0 0 90>}]
 dyn_hair_c_02 [{shapeType=Capsule, length=2, radius=2.5, shapeName=dyn_hair_c_02_capsule, center=<3 0 1>, rotation=<rot 0 0 90>}]
 dyn_hair_c_03 [{shapeType=Capsule, length=2, radius=2.5, shapeName=dyn_hair_c_03_capsule, center=<1.8 0 1>, rotation=<rot 0 0 90>}]
 dyn_hair_c_Master [{shapeType=Capsule, length=0.505, radius=0.505, shapeName=dyn_hair_c_Master_capsule, rotation=<rot 0 -90 0>}]
 dyn_hair_l_A_01 [{shapeType=Capsule, length=2, radius=2.5, shapeName=dyn_hair_l_A_01_capsule, rotation=<rot 0 0 90>}]
 dyn_hair_l_A_02 [{shapeType=Capsule, length=2, radius=2.5, shapeName=dyn_hair_l_A_02_capsule, center=<3 0 1>, rotation=<rot 0 0 90>}]
 dyn_hair_l_A_03 [{shapeType=Capsule, length=2, radius=2.5, shapeName=dyn_hair_l_A_03_capsule, center=<1.8 0 1>, rotation=<rot 0 0 90>}]
 dyn_hair_l_A_Master [{shapeType=Capsule, length=0.505, radius=0.505, shapeName=dyn_hair_l_A_Master_capsule, rotation=<rot 0 -90 0>}]
 dyn_hair_l_B_01 [{shapeType=Capsule, length=2, radius=2.5, shapeName=dyn_hair_l_B_01_capsule, rotation=<rot 0 0 90>}]
 dyn_hair_l_B_02 [{shapeType=Capsule, length=2, radius=2.5, shapeName=dyn_hair_l_B_02_capsule, center=<3 0 1>, rotation=<rot 0 0 90>}]
 dyn_hair_l_B_03 [{shapeType=Capsule, length=2, radius=2.5, shapeName=dyn_hair_l_B_03_capsule, center=<1.8 0 1>, rotation=<rot 0 0 90>}]
 dyn_hair_l_B_Master [{shapeType=Capsule, length=0.505, radius=0.505, shapeName=dyn_hair_l_B_Master_capsule, rotation=<rot 0 -90 0>}]
 dyn_hair_r_A_01 [{shapeType=Capsule, length=2, radius=2.5, shapeName=dyn_hair_r_A_01_capsule, rotation=<rot 0 0 90>}]
 dyn_hair_r_A_02 [{shapeType=Capsule, length=2, radius=2.8, shapeName=dyn_hair_r_A_02_capsule, center=<3 0 1>, rotation=<rot 0 0 90>}]
 dyn_hair_r_A_03 [{shapeType=Capsule, length=2, radius=2.7, shapeName=dyn_hair_r_A_03_capsule, center=<1.8 0 1>, rotation=<rot 0 0 90>}]
 dyn_hair_r_A_Master [{shapeType=Capsule, length=0.505, radius=0.505, shapeName=dyn_hair_r_A_Master_capsule, rotation=<rot 0 -90 0>}]
 dyn_hair_r_B_01 [{shapeType=Capsule, length=2, radius=2.5, shapeName=dyn_hair_r_B_01_capsule, rotation=<rot 0 0 90>}]
 dyn_hair_r_B_02 [{shapeType=Capsule, length=2, radius=2.5, shapeName=dyn_hair_r_B_02_capsule, center=<3 0 1>, rotation=<rot 0 0 90>}]
 dyn_hair_r_B_03 [{shapeType=Capsule, length=2, radius=2.7, shapeName=dyn_hair_r_B_03_capsule, center=<1.8 0 1>, rotation=<rot 0 0 90>}]
 dyn_hair_r_B_Master [{shapeType=Capsule, length=0.505, radius=0.505, shapeName=dyn_hair_r_B_Master_capsule, rotation=<rot 0 -90 0>}]
 head [{shapeType=Capsule, length=4, radius=7, rotation=<rot 90 0 0>}]
 spine_05 [{shapeType=Capsule, length=5.5321, radius=13.6002, shapeName=spine_05_capsule, center=<-2.3746 0.571 0>, rotation=<rot 0 -90 0>}]
 upperarm_l [{shapeType=Capsule, length=25, radius=7.5, shapeName=upperarm_l_capsule, center=<13.1232 0 -1.5489>, rotation=<rot 90 0 0>}]
 upperarm_r [{shapeType=Capsule, length=25, radius=7.5, shapeName=upperarm_l_capsule, center=<-15.7589 0 1.689>, rotation=<rot 90 180 0>}]
}
constraints [
 {
  bone1Name=dyn_hair_c_Master
  bone2Name=head
  swing1LimitDegrees=45
  swing1Motion=Free
  swing2LimitDegrees=45
  swing2Motion=Free
  twistLimitDegrees=45
  twistMotion=Free
 }
 {bone1Name=dyn_hair_c_01, bone2Name=dyn_hair_c_Master, swing1LimitDegrees=10, swing2LimitDegrees=20, twistLimitDegrees=0}
 {bone1Name=dyn_hair_c_02, bone2Name=dyn_hair_c_01, swing1LimitDegrees=10, swing2LimitDegrees=30, twistLimitDegrees=0}
 {bone1Name=dyn_hair_c_03, bone2Name=dyn_hair_c_02, swing1LimitDegrees=10, swing2LimitDegrees=30, twistLimitDegrees=0}
 {
  bone1Name=dyn_hair_l_A_Master
  bone2Name=head
  swing1LimitDegrees=45
  swing1Motion=Free
  swing2LimitDegrees=45
  swing2Motion=Free
  twistLimitDegrees=45
  twistMotion=Free
 }
 {bone1Name=dyn_hair_l_A_01, bone2Name=dyn_hair_l_A_Master, swing1LimitDegrees=10, swing2LimitDegrees=20, twistLimitDegrees=0}
 {bone1Name=dyn_hair_l_A_03, bone2Name=dyn_hair_l_A_02, swing1LimitDegrees=10, swing2LimitDegrees=10, twistLimitDegrees=0}
 {bone1Name=dyn_hair_l_A_02, bone2Name=dyn_hair_l_A_01, swing1LimitDegrees=10, swing2LimitDegrees=10, twistLimitDegrees=0}
 {
  bone1Name=dyn_hair_l_B_Master
  bone2Name=head
  swing1LimitDegrees=45
  swing1Motion=Free
  swing2LimitDegrees=45
  swing2Motion=Free
  twistLimitDegrees=45
  twistMotion=Free
 }
 {bone1Name=dyn_hair_l_B_01, bone2Name=dyn_hair_l_B_Master, swing1LimitDegrees=10, swing2LimitDegrees=20, twistLimitDegrees=0}
 {bone1Name=dyn_hair_l_B_02, bone2Name=dyn_hair_l_B_01, swing1LimitDegrees=10, swing2LimitDegrees=30, twistLimitDegrees=0}
 {bone1Name=dyn_hair_l_B_03, bone2Name=dyn_hair_l_B_02, swing1LimitDegrees=10, swing2LimitDegrees=30, twistLimitDegrees=0}
 {
  bone1Name=dyn_hair_r_A_Master
  bone2Name=head
  swing1LimitDegrees=45
  swing1Motion=Free
  swing2LimitDegrees=45
  swing2Motion=Free
  twistLimitDegrees=45
  twistMotion=Free
 }
 {bone1Name=dyn_hair_r_A_01, bone2Name=dyn_hair_r_A_Master, swing1LimitDegrees=10, swing2LimitDegrees=20, twistLimitDegrees=0}
 {bone1Name=dyn_hair_r_A_02, bone2Name=dyn_hair_r_A_01, swing1LimitDegrees=10, swing2LimitDegrees=10, twistLimitDegrees=0}
 {bone1Name=dyn_hair_r_A_03, bone2Name=dyn_hair_r_A_02, swing1LimitDegrees=10, swing2LimitDegrees=10, twistLimitDegrees=0}
 {
  bone1Name=dyn_hair_r_B_Master
  bone2Name=head
  swing1LimitDegrees=45
  swing1Motion=Free
  swing2LimitDegrees=45
  swing2Motion=Free
  twistLimitDegrees=45
  twistMotion=Free
 }
 {bone1Name=dyn_hair_r_B_01, bone2Name=dyn_hair_r_B_Master, swing1LimitDegrees=10, swing2LimitDegrees=20, twistLimitDegrees=0}
 {bone1Name=dyn_hair_r_B_02, bone2Name=dyn_hair_r_B_01, swing1LimitDegrees=10, swing2LimitDegrees=30, twistLimitDegrees=0}
 {bone1Name=dyn_hair_r_B_03, bone2Name=dyn_hair_r_B_02, swing1LimitDegrees=10, swing2LimitDegrees=30, twistLimitDegrees=0}
 {
  bone1Name=head
  bone2Name=spine_05
  swing1LimitDegrees=45
  swing1Motion=Free
  swing2LimitDegrees=45
  swing2Motion=Free
  twistLimitDegrees=45
  twistMotion=Free
 }
 {
  bone1Name=clavicle_r
  bone2Name=spine_05
  swing1LimitDegrees=45
  swing1Motion=Free
  swing2LimitDegrees=45
  swing2Motion=Free
  twistLimitDegrees=45
  twistMotion=Free
 }
 {
  bone1Name=clavicle_l
  bone2Name=spine_05
  swing1LimitDegrees=45
  swing1Motion=Free
  swing2LimitDegrees=45
  swing2Motion=Free
  twistLimitDegrees=45
  twistMotion=Free
 }
 {
  bone1Name=upperarm_r
  bone2Name=clavicle_r
  swing1LimitDegrees=45
  swing1Motion=Free
  swing2LimitDegrees=45
  swing2Motion=Free
  twistLimitDegrees=45
  twistMotion=Free
 }
 {
  bone1Name=upperarm_l
  bone2Name=clavicle_l
  swing1LimitDegrees=45
  swing1Motion=Free
  swing2LimitDegrees=45
  swing2Motion=Free
  twistLimitDegrees=45
  twistMotion=Free
 }
]
